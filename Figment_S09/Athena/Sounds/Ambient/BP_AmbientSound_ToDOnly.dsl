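; BP_AmbientSound_ToDOnly is an ambient sound Blueprint that plays only during chosen time-of-day phases.
; On BeginPlay it waits asynchronously for the time-of-day manager, stores it, and binds to its day-phase change event.
; Each phase change checks the new phase against the AllowedTimeOfDayPhases array, fading the audio component in on a match and out otherwise, with linear fades.
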(event EventBeginPlay
  (bind _timeofdaymanager (Class|FortAsyncActionTimeOfDayManagerReady|GetTimeOfDayManagerReady)
    (:OnReady
      (bind _output_get (Variables|Default|SetTimeOfDayManager (Utilities|Casting|CastToFortTimeOfDayManager _timeofdaymanager)))
      (EventDispatchers|BindEventtoOnDayPhaseChangeEvent _output_get (delegate Custom|On%20Day%20Phase%20Changed)))))

(event Custom|On%20Day%20Phase%20Changed
  (params (CurrentDayPhase :type EFortDayPhase%20Enum :default "Morning") (PreviousDayPhase :type EFortDayPhase%20Enum :default "Morning") (bAtCreation :type Boolean))
  (bind _audiocomponent (Variables|Sound|GetAudioComponent))
  (if (Utilities|Array|ContainsItem (Variables|Default|GetAllowedTimeOfDayPhases) CurrentDayPhase)
    (Audio|Components|Audio|FadeIn _audiocomponent 5.0 1.0 0.0 "Linear" :!class /Script/Engine.AudioComponent)
    (else
      (Audio|Components|Audio|FadeOut _audiocomponent 5.0 0.0 "Linear" :!class /Script/Engine.AudioComponent))))

(fn ConstructionScript)
